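# SpeedTree material function that converts a tangent-space normal map into a world-space normal facing the camera, for smoother billboard lighting.
# The normal's red and green channels, with zero Z, pass through a default vector Transform.
# Blue is negated and scales the normalized camera-relative world position, replacing the surface normal with a per-pixel direction toward the camera.
# The two parts sum into the NormalWS output without renormalizing.
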MaterialFunction /Engine/Functions/Engine_MaterialFunctions01/SpeedTree/SpeedTreeBillboardNormals
graph {
 properties {bExposeToLibrary=true, description=Move normals into screen space for smoother billboard lighting, libraryCategoriesText [SpeedTree]}
 nodes [
  n1 = FunctionInput() {description=Tangent space normal map, inputName=Normal}
  n2 = AppendVector(A=ComponentMask(None=@n1) {g=true, r=true}, B=Constant())
  n3 = Multiply(A=ComponentMask(None=@n1) {b=true}) {constB=-1}
  n4 = Normalize(VectorInput=WorldPosition() {worldPositionShaderOffset=WPT_CameraRelative}.XYZ)
  FunctionOutput(None=Add(A=Transform(None=@n2), B=Multiply(A=@n3, B=@n4))) {outputName=NormalWS}
 ]
}
