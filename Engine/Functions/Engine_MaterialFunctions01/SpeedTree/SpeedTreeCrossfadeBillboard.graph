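# SpeedTree material function that crossfades between billboard faces by fading opacity with viewing angle.
# The CustomUV output takes the minimum of the vertex normal's dots with camera-to-object and view-axis directions, raises it to the fourth power, adds a UV2.x-driven bias and saturates.
# OpacityMask multiplies input opacity by UV2.x, implying the fade is computed per vertex through customized UV channel 2.
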MaterialFunction /Engine/Functions/Engine_MaterialFunctions01/SpeedTree/SpeedTreeCrossfadeBillboard
graph {
 properties {bExposeToLibrary=true, description=Handles crossfading between billboard faces, libraryCategoriesText [SpeedTree]}
 nodes [
  n1 = TextureCoordinate() {coordinateIndex=2}
  FunctionOutput(None=Multiply(A=ComponentMask(None=@n1) {r=true}, B=FunctionInput() {description=Original opacity, inputName=Opacity, inputType=FunctionInput_Scalar})) {outputName=OpacityMask}
  n2 = Subtract(A=ObjectPositionWS(), B=CameraPositionWS())
  n3 = VertexNormalWS()
  n4 = Constant3Vector() {constant=<rgba 0 0 1 0>}
  n5 = Min(A=DotProduct(A=Normalize(VectorInput=@n2), B=@n3), B=DotProduct(A=@n3, B=Transform(None=@n4) {transformSourceType=TRANSFORMSOURCE_View}))
  n6 = Multiply(A=@n5, B=@n5)
  n7 = TextureCoordinate() {coordinateIndex=2}
  n8 = Add(A=Multiply(A=@n6, B=@n6), B=LinearInterpolate(Alpha=ComponentMask(None=@n7) {r=true}) {constA=0.18, constB=0.38}) {constB=0.1}
  FunctionOutput(None=AppendVector(A=Saturate(None=@n8), B=Constant())) {outputName=CustomUV, sortPriority=1}
 ]
}
