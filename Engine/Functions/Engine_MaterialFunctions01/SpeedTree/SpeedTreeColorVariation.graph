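# SpeedTree material function that adds per-object or per-instance color variation to BaseColor, breaking up repeated instances in large forests.
# Scaled PerInstanceRandom and object world position are summed and wrapped with Frac into a pseudo-random vector.
# A dot product of its RG and GB swizzles, biased and scaled, sets the strength, multiplied by the Amount input.
# That strength scales the normalized random vector, which is added to BaseColor.
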MaterialFunction /Engine/Functions/Engine_MaterialFunctions01/SpeedTree/SpeedTreeColorVariation
graph {
 properties {
  bExposeToLibrary=true
  description=Adds a color variation per object/instance to break up instances in a large forest
  libraryCategoriesText [Misc,SpeedTree]
 }
 nodes [
  n1 = FunctionInput() {bUsePreviewValueAsDefault=true, inputName=Amount, inputType=FunctionInput_Scalar, sortPriority=1, previewValue=<0.05 0 0 0>}
  n2 = Constant3Vector() {constant=<rgba 100 10 1 0>}
  n3 = PerInstanceRandom()
  n4 = ObjectPositionWS()
  n5 = Constant() {r=0.01}
  n6 = Frac(None=Add(A=Multiply(A=@n2, B=@n3), B=Multiply(A=@n4, B=@n5)))
  n7 = ComponentMask(None=@n6) {g=true, r=true}
  n8 = ComponentMask(None=@n6) {b=true, g=true}
  n9 = Multiply(A=@n1, B=ConstantBiasScale(None=DotProduct(A=@n7, B=@n8)) {bias=-0.5, scale=2})
  n10 = Normalize(VectorInput=@n6)
  FunctionOutput(None=Add(A=Multiply(A=@n9, B=@n10), B=FunctionInput() {inputName=BaseColor}))
 ]
}
